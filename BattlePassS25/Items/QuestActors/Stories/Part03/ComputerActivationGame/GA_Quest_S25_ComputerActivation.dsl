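; Gameplay ability that tracks a player's game state for the Season 25 Stories Part 3 computer activation quest minigame.
; On activation, the server applies the game counter effect to the executing player pawn's ability system to start the count.
; On end, the server removes GameStarted and GameFinished tagged effects, then casts the granting effect's causer to the game prop.
; No cleanup call follows that cast, despite the comment describing one.
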(event Ability|EventActivateAbility
  ;[1 1 1] On game start, apply the first stack of the game counter to begin
  (|CallParentFunction)
  (if (Networking|IsServer)
    (bind _fortplayerpawn (CallFunction|GetExecutingPlayerPawn))
    (Utilities|IsValid (Ability|GetFortAbilitySystemComponent _fortplayerpawn)
      (:Is%20Valid
        (GameplayEffects|ApplyGameplayEffectToSelf :self (Ability|GetFortAbilitySystemComponent _fortplayerpawn) :GameplayEffectClass "/BattlePassS25/Items/QuestActors/Stories/Part03/ComputerActivationGame/GE_Quest_S25_CompActivate_GameCounter.GE_Quest_S25_CompActivate_GameCounter_C" :Level 0.0)))))

(event Ability|EventOnEndAbility
  (params (bWasCancelled :type Boolean))
  ;[1 1 1] On End Ability - Remove any active GEs tracking the game state and call cleanup on the prop
  (|CallParentFunction bWasCancelled)
  (if (Networking|IsServer)
    (bind _fortplayerpawn (CallFunction|GetExecutingPlayerPawn))
    (Utilities|IsValid (Ability|GetFortAbilitySystemComponent _fortplayerpawn)
      (:Is%20Valid
        (GameplayEffects|RemoveActiveEffectswithTags (Ability|GetFortAbilitySystemComponent _fortplayerpawn) "(GameplayTags=((TagName=\"Granted.Quest.S25.P03.Q04.CompActivate.GameStarted\"),(TagName=\"Granted.Quest.S25.P03.Q04.CompActivate.GameFinished\")))")
        (Utilities|Casting|CastToBP_S25_ComputerActivation_Game (Ability|EffectContext|GetEffectCauser (Ability|GetGrantedbyEffectContext)))))))

(event Ability|EventFailedToActivatePassiveAbility
  (params (ActorInfo :type Gameplay%20Ability%20Actor%20Info%20Structure))
  (|CallParentFunction))

(fn GetExecutingPlayerPawn
  (bind _asfort_player_pawn (Utilities|Casting|CastToFortPlayerPawn (Ability|GetAvatarActorfromActorInfo))
    (:then
      (return _asfort_player_pawn))
    (:CastFailed
      (return 0))))
